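# Masked, two-sided subsurface material for interactive leaf effects, enabled for Niagara mesh particles.
# The oak leaf texture is fully desaturated and tinted by particle color, so the effect controls leaf hue.
# World-position curl noise, scaled by the "World Position Amount" dynamic parameter, provides an offset for deforming leaves.
# Subsurface color and SSS strength are exposed as parameters, alongside an inverted blue channel from the leaf normal map.
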Material /Game/Effects/Fort_Effects/World/Enviroment/Leaf_Interactive/M_InteractableFX_Leaf
graph {
 properties {
  bUsedWithNiagaraMeshParticles=true
  blendMode=BLEND_Masked
  pixelDepthOffsetMode=PDOM_Legacy
  shadingModel=MSM_Subsurface
  twoSided=true
  usedShadingModels=Subsurface
 }
 nodes [
  Multiply(
   A=Desaturation(None=TextureSample() {texture=/Game/Effects/Fort_Effects/World/Destruction/Textures/T_Oak_Tree_Leaf}.RGB, Fraction=Constant() {r=1})
   B=ParticleColor().RGB
  )
  n1 = WorldPosition()
  n2 = ScalarParameter() {defaultValue=0.06, parameterName=WPO Noise Size}
  Multiply(
   A=DynamicParameter() {paramNames [World Position Amount, Param2, Param3, Param4]}.World Position Amount
   B=VectorNoise(World Position=Multiply(A=@n1.XYZ, B=@n2)) {noiseFunction=VNF_CurlALU}
  )
  OneMinus(None=TextureSample() {samplerType=SAMPLERTYPE_Normal, texture=/Game/Effects/Fort_Effects/World/Destruction/Textures/T_Oak_Tree_Leaf_N_01}.B)
  ScalarParameter() {defaultValue=0.85, parameterName=SSS}
  VectorParameter() {parameterName=Subsurface Color, defaultValue=<rgba 0.349 0.0876 0 0>}
 ]
}
